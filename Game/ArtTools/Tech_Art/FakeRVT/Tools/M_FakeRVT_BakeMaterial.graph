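# Unlit tool material that, judging by its name, bakes Apollo landscape runtime virtual texture data for the FakeRVT system.
# It samples the landscape and height RVTs through UVs scaled by a Zoom parameter.
# One chain packs specular into two channels with divided world height in the third.
# Other chains offset the normal, build an X/Y-driven eighth-tile UV, and remap height, with no consumers shown.
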Material /Game/ArtTools/Tech_Art/FakeRVT/Tools/M_FakeRVT_BakeMaterial
graph {
 properties {pixelDepthOffsetMode=PDOM_Legacy, shadingModel=MSM_Unlit, usedShadingModels=Unlit}
 nodes [
  n1 = Multiply(A=TextureCoordinate(), B=ScalarParameter() {defaultValue=1, parameterName=Zoom})
  n2 = RuntimeVirtualTextureSampleParameter(UVs=@n1) {materialType=BaseColor_Normal_Specular_YCoCg, parameterName=Param, virtualTexture=/Game/Athena/Apollo/Maps/Landscape/VT/RVT_Apollo_Landscape}
  Add(A=@n2.Normal, B=Constant3Vector() {constant=<rgba 0.5 0.5 0 0>})
  n3 = ScalarParameter() {defaultValue=5, parameterName=X}
  n4 = ScalarParameter() {defaultValue=5, parameterName=Y}
  Add(
   A=TextureCoordinate() {uTiling=0.125, vTiling=0.125}
   B=Divide(A=MaterialFunctionCall(X=@n3, Y=@n4) {materialFunction=&MakeFloat2}.Result) {constB=8}
  )
  n5 = RuntimeVirtualTextureSampleParameter(UVs=@n1) {materialType=WorldHeight, parameterName=1234, virtualTexture=/Game/Athena/Apollo/Maps/Landscape/VT/RVT_Apollo_Landscape_Height}
  MaterialExpressionAppend3Vector_0 = MaterialXAppend3Vector!(A=@n2.Specular, B=@n2.Specular, C=Divide(A=@n5.WorldHeight) {constB=15000})
  Constant3Vector() {constant=<rgba 0.5 0.5 1 0>}
  Constant()
  Constant() {r=32513.8008}
  InverseLinearInterpolate(Value=@n5.WorldHeight) {constA=-3100, constB=15000}
  ScalarParameter() {defaultValue=10}
  TextureSample()
 ]
}
