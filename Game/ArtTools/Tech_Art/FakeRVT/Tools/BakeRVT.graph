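# BakeRVT is an Editor Utility Widget tool that bakes the FakeRVT render target, apparently standing in for a runtime virtual texture.
# On every Tick it draws the M_FakeRVT_BakeMaterial material into the FRVT_RT render target, so the target refreshes continuously while the widget runs.
# The UI is only an empty Canvas Panel, and the Construct and PreConstruct events are unused.
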EditorUtilityWidgetBlueprint /Game/ArtTools/Tech_Art/FakeRVT/Tools/BakeRVT
graphs [EventGraph {nodes [
   Event Tick()
   > DrawMaterialToRenderTarget!(
    execute
    WorldContextObject:Object Reference
    TextureRenderTarget:Texture Render Target 2D Object Reference=&/FRVT_RT
    Material:Material Interface Object Reference=&/M_FakeRVT_BakeMaterial
   ) -> (then)
   Event PreConstruct()
   Event Construct()
  ]}]
variables []
widgets {
 info {rootWidgetClass=&CanvasPanel, widgetCount=1}
 widgets [CanvasPanel_0 {widgetClassPath=&CanvasPanel, widget=CanvasPanel_0}]
}
