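; Terrain capture tool Blueprint in the FakeRVT art tools that sets up a top-down SceneCapture2D of the landscape.
; Its construction script centers the actor in XY on the ActorToCapture bounds at a fixed height and applies a fixed, very wide ortho width.
; It then fills the capture's show-only list with every LandscapeStreamingProxy in the level.
; BeginPlay, Tick and overlap events are empty, so all setup happens at construction.
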(event EventBeginPlay)

(event Collision|EventActorBeginOverlap
  (params (OtherActor :type Actor%20Object%20Reference)))

(event EventTick
  (params (DeltaSeconds :type Float%20%28single-precision%29)))

(fn ConstructionScript
  (bind _origin (Collision|GetActorBounds (Variables|Default|GetActorToCapture) false true))
  (bind (_origin_x _origin_y) (Math|Vector|BreakVector _origin))
  (Transformation|SetActorLocation 0 (Math|Vector|MakeVector _origin_x _origin_y 1e+04) false true)
  (Variables|/Script/Engine.SceneCaptureComponent2D|SetOrthoWidth 260096.0 (Variables|Default|GetSceneCaptureComponent2D))
  (bind _outactors (Actor|GetAllActorsOfClass "/Script/Landscape.LandscapeStreamingProxy"))
  (Variables|/Script/Engine.SceneCaptureComponent|SetShowOnlyActors _outactors (Variables|Default|GetSceneCaptureComponent2D)))
